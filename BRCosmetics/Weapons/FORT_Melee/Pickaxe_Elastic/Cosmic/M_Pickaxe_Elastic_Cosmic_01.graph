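# Translucent unlit effect material for the Cosmic Elastic pickaxe, usable on skeletal and Niagara meshes.
# Dynamic-parameter primary and secondary colors blend through a curve-atlas gradient, with three panning radial-UV detail layers added as emissive.
# Opacity combines a gradient plus panning mask, a noise-distorted radial bottom falloff and a shape texture, softened by depth and view-angle fades.
# Panning noise pushes vertices along normals, masked by texture and vertical gradients.
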Material /BRCosmetics/Weapons/FORT_Melee/Pickaxe_Elastic/Cosmic/M_Pickaxe_Elastic_Cosmic_01
graph {
 properties {
  bUsedWithMeshDeformer=true
  bUsedWithNiagaraMeshParticles=true
  bUsedWithSkeletalMesh=true
  bUsedWithSplineMeshes=true
  blendMode=BLEND_Translucent
  pixelDepthOffsetMode=PDOM_Legacy
  shadingModel=MSM_Unlit
  usedShadingModels=Unlit
 }
 nodes [
  n1 = DynamicParameter() {parameterIndex=1, paramNames [SecondaryR,SecondaryG,SecondaryB,SecondaryA]}
  n2 = AppendVector(A=AppendVector(A=@n1.SecondaryR, B=@n1.SecondaryR), B=@n1.SecondaryB)
  AppendVector(A=@n2)
  n3 = DynamicParameter() {paramNames [PrimaryR,PrimaryG,PrimaryB,PrimaryA]}
  n4 = AppendVector(A=AppendVector(A=@n3.PrimaryR, B=@n3.PrimaryR), B=@n3.PrimaryB)
  AppendVector(A=@n4)
  n5 = TextureCoordinate()
  n6 = OneMinus(None=ComponentMask(None=@n5) {g=true})
  n7 = ComponentMask(None=@n5) {r=true}
  n8 = Multiply(A=Abs(None=Subtract(A=@n7) {constB=0.5})) {constB=0.6}
  n9 = ComponentMask(None=TextureCoordinate()) {g=true}
  n10 = CurveAtlasRowParameter(CurveTime=LinearInterpolate(A=Subtract(A=@n6, B=@n8), B=OneMinus(None=@n9), Alpha=ScalarParameter() {defaultValue=1, parameterName=Slope}) {constA=1}) {
   atlas=/Game/Athena/LinearColorCurves/TempAthenaCurveAtlas
   curve=/Game/Weapons/FORT_Melee/Pickaxe_Elastic/Cosmic/Curve_Pickaxe_Elastic_Cosmic_01
   defaultValue=12
  }
  n11 = Multiply(A=TextureCoordinate(), B=AppendVector(A=Constant(), B=ScalarParameter() {defaultValue=5, group=Mask, parameterName=UV-Tiling_Grad}))
  n12 = AppendVector(A=Constant(), B=ScalarParameter() {group=Mask, parameterName=UV-Offset_Grad})
  n13 = Clamp(None=ComponentMask(None=Add(A=@n11, B=@n12)) {g=true})
  LinearInterpolate(A=@n13, B=@n10.A, Alpha=ScalarParameter() {defaultValue=1, group=Mask, parameterName=UseColorCurveForMaskGrad})
  n14 = Desaturation(
   None=Multiply(A=@n4, B=ScalarParameter() {defaultValue=1, group=Color, parameterName=Primary Color Intensity})
   Fraction=ScalarParameter() {group=Color, parameterName=Primary Color Desaturation}
  )
  n15 = MaterialFunctionCall(Hue Shift Percentage=ScalarParameter() {group=Color, parameterName=Detail_Hue}, Texture=@n14) {materialFunction=&HueShift}
  n16 = TextureCoordinate() {coordinateIndex=1, uTiling=2}
  n17 = ComponentMask(None=@n16) {r=true}
  n18 = Add(A=ComponentMask(None=@n16) {g=true}, B=ScalarParameter() {group=Detail, parameterName=Detail_Center})
  n19 = MaterialFunctionCall(Vector or UVs=Subtract(A=AppendVector(A=@n17, B=@n18), B=Constant() {r=1})) {materialFunction=&VectorToRadialValue}
  n20 = ScalarParameter() {defaultValue=1, group=Detail, parameterName=NoiseTexture_Tile_U}
  n21 = ScalarParameter() {defaultValue=1, group=Detail, parameterName=NoiseTexture_Tile_V}
  n22 = Panner(
   Coordinate=Multiply(A=@n19.Radial Coordinates, B=AppendVector(A=@n20, B=@n21))
   Speed=AppendVector(A=ScalarParameter() {group=Detail, parameterName=NoiseTexture_Pan_U}, B=ScalarParameter() {group=Detail, parameterName=NoiseTexture_Pan_V})
  ) {speedY=0.3}
  n23 = Multiply(
   A=ScalarParameter() {defaultValue=1, group=Detail, parameterName=NoiseTexture_Intensity}
   B=Multiply(
    A=ComponentMask(None=@n19.Linear Distance) {r=true}
    B=TextureSampleParameter2D(UVs=@n22) {group=Detail, parameterName=NoiseTexture, texture=/Game/Packages/Fortress_SharedMaterials/Textures/T_Ravens_SpecularReflection}.R
   )
  )
  n24 = ScalarParameter() {defaultValue=1, group=Detail, parameterName=NoiseTexture_Power}
  n25 = ScalarParameter() {defaultValue=1, group=Color, parameterName=ColorCurve_Power}
  n26 = Clamp(None=Multiply(A=Power(Base=@n10.R, Exp=@n25), B=ScalarParameter() {defaultValue=1, group=Color, parameterName=ColorCurve_Intensity}))
  n27 = ComponentMask(None=@n19.Linear Distance) {r=true}
  n28 = AppendVector(
   A=ScalarParameter() {defaultValue=1, group=Detail, parameterName=LineTexture_Tile_U}
   B=ScalarParameter() {defaultValue=1, group=Detail, parameterName=LineTexture_Tile_V}
  )
  n29 = ScalarParameter() {group=Detail, parameterName=LineTexture_Pan_U}
  n30 = ScalarParameter() {group=Detail, parameterName=LineTexture_Pan_V}
  n31 = TextureSampleParameter2D(UVs=Panner(Coordinate=Multiply(A=@n19.Radial Coordinates, B=@n28), Speed=AppendVector(A=@n29, B=@n30)) {speedY=0.3}) {group=Detail, parameterName=LineTexture, texture=&T_Ravens_SpecularReflection}
  n32 = Multiply(
   A=@n15.Result
   B=Multiply(A=ScalarParameter() {defaultValue=1, group=Detail, parameterName=LineTexture_Intensity}, B=Multiply(A=@n27, B=@n31.R))
  )
  n33 = Power(Base=@n26, Exp=ScalarParameter() {defaultValue=1, group=Detail, parameterName=LineTexture_Power})
  n34 = ComponentMask(None=@n19.Linear Distance) {r=true}
  n35 = AppendVector(
   A=ScalarParameter() {defaultValue=1, group=Detail, parameterName=DetailTexture_Tile_U}
   B=ScalarParameter() {defaultValue=1, group=Detail, parameterName=DetailTexture_Tile_V}
  )
  n36 = ScalarParameter() {group=Detail, parameterName=DetailTexture_Pan_U}
  n37 = ScalarParameter() {group=Detail, parameterName=DetailTexture_Pan_V}
  n38 = TextureSampleParameter2D(UVs=Panner(Coordinate=Multiply(A=@n19.Radial Coordinates, B=@n35), Speed=AppendVector(A=@n36, B=@n37)) {speedY=0.3}) {group=Detail, parameterName=DetailTexture, texture=&T_Ravens_SpecularReflection}
  n39 = Multiply(
   A=@n15.Result
   B=Multiply(A=ScalarParameter() {defaultValue=1, group=Detail, parameterName=DetailTexture_Intensity}, B=Multiply(A=@n34, B=@n38.R))
  )
  n40 = Power(Base=@n26, Exp=ScalarParameter() {defaultValue=1, group=Detail, parameterName=DetailTexture_Power})
  n41 = Add(A=Multiply(A=Multiply(A=@n15.Result, B=@n23), B=Power(Base=@n26, Exp=@n24)), B=Add(A=Multiply(A=@n32, B=@n33), B=Multiply(A=@n39, B=@n40)))
  n42 = ScalarParameter() {defaultValue=1, group=Color, parameterName=Secondary Color Intensity}
  n43 = LinearInterpolate(
   A=Desaturation(None=Multiply(A=@n42, B=@n2), Fraction=ScalarParameter() {group=Color, parameterName=Secondary Color Desaturation})
   B=@n14
   Alpha=@n26
  )
  MaterialFunctionCall(Emissive=Multiply(A=Add(A=@n41, B=@n43), B=@n10.B)) {materialFunction=&MF_FX_EmissiveAdjustments}
  n44 = TextureCoordinate() {coordinateIndex=1}
  n45 = ScalarParameter() {defaultValue=2.9, group=Bottom, parameterName=Bottom_Width}
  n46 = AppendVector(
   A=Multiply(A=ComponentMask(None=@n44) {r=true}, B=Multiply(A=@n45) {constB=2})
   B=Add(A=ComponentMask(None=@n44) {g=true}, B=ScalarParameter() {defaultValue=1.2, group=Bottom, parameterName=Bottom_Center})
  )
  n47 = ComponentMask(None=MaterialFunctionCall(Vector or UVs=Subtract(A=@n46, B=@n45)) {materialFunction=&VectorToRadialValue}.Linear Distance) {r=true}
  n48 = ScalarParameter() {defaultValue=9, group=Bottom, parameterName=Bottom_Power}
  n49 = Subtract(
   A=Multiply(A=Power(Base=@n47, Exp=@n48), B=ScalarParameter() {defaultValue=4, group=Bottom, parameterName=Bottom_Mul})
   B=ScalarParameter() {defaultValue=30, group=Bottom, parameterName=Bottom_Add}
  )
  n50 = Multiply(
   A=TextureCoordinate() {coordinateIndex=1}
   B=AppendVector(
    A=ScalarParameter() {defaultValue=2, group=Distortion, parameterName=Distortion_Tile_U}
    B=ScalarParameter() {defaultValue=0.5, group=Distortion, parameterName=Distortion_Tile_V}
   )
  )
  n51 = AppendVector(
   A=ScalarParameter() {defaultValue=-0.5, group=Distortion, parameterName=Distortion_Pan_U}
   B=ScalarParameter() {defaultValue=0.3, group=Distortion, parameterName=Distortion_Pan_V}
  )
  n52 = Multiply(
   A=TextureSample(UVs=Panner(Coordinate=@n50, Speed=@n51) {speedY=0.3}) {texture=&T_NoisePack_01}.R
   B=ScalarParameter() {defaultValue=20, group=Distortion, parameterName=Distortion_Intensity}
  ) {constB=0.1}
  n53 = ScalarParameter() {defaultValue=1, group=Mask, parameterName=Mask_Power}
  n54 = Clamp(None=Multiply(A=Power(Base=@n13, Exp=@n53), B=ScalarParameter() {defaultValue=1, group=Mask, parameterName=Mask_Grad_Hardness}))
  n55 = TextureCoordinate() {uTiling=-1, vTiling=-1}
  n56 = OneMinus(None=ComponentMask(None=@n55) {r=true})
  n57 = ComponentMask(None=@n55) {g=true}
  n58 = LinearInterpolate(
   A=TextureCoordinate()
   B=AppendVector(A=Multiply(A=@n56, B=@n57), B=@n57)
   Alpha=ScalarParameter() {defaultValue=0.5, group=Mask, parameterName=Mask_Slope}
  )
  n59 = AppendVector(
   A=ScalarParameter() {defaultValue=1, group=Mask, parameterName=MaskTexture_Tile_U}
   B=ScalarParameter() {defaultValue=0.2, group=Mask, parameterName=MaskTexture_Tile_V}
  )
  n60 = ScalarParameter() {defaultValue=1, group=Mask, parameterName=MaskTexture_Pan_U}
  n61 = ScalarParameter() {defaultValue=-0.3, group=Mask, parameterName=MaskTexture_Pan_V}
  n62 = TextureSampleParameter2D(UVs=Panner(Coordinate=Multiply(A=@n58, B=@n59), Speed=AppendVector(A=@n60, B=@n61)) {speedY=0.3}) {
   group=Mask
   parameterName=MaskTexture
   samplerType=SAMPLERTYPE_Normal
   texture=/Game/Global/Textures/Default/Tileable_Patterns/T_StormCloudDetail01_N
  }
  n63 = Clamp(None=Multiply(A=ScalarParameter() {defaultValue=0.3, group=Mask, parameterName=Mask_Intensity}, B=@n62.R))
  n64 = Multiply(A=MaterialFunctionCall!(Base=@n54, Add=@n63) {materialFunction=/Engine/Functions/Engine_MaterialFunctions02/Math/MultiplyAdd}.Result)
  n65 = Multiply(A=@n64, B=Clamp(None=Add(A=@n49, B=@n52)))
  n66 = TextureCoordinate() {coordinateIndex=1}
  n67 = TextureSample(UVs=@n66) {texture=&/T_Pickaxe_Elastic_Mask_01}
  n68 = Clamp(None=Multiply(A=Power(Base=@n67.R) {constExponent=1.21}, B=ScalarParameter() {defaultValue=22, group=Mask, parameterName=ShapeFade}))
  Multiply(
   A=DepthFade(Opacity=Multiply(A=@n65, B=@n68), FadeDistance=ScalarParameter() {parameterName=DepthFade})
   B=MaterialFunctionCall("FadeSensitivity (optional)"=ScalarParameter() {defaultValue=1, parameterName=AngleFade}) {materialFunction=&FadeBasedOnViewAngle}.Result
  )
  n69 = VertexNormalWS()
  n70 = Multiply(
   A=TextureCoordinate() {coordinateIndex=1}
   B=AppendVector(
    A=ScalarParameter() {defaultValue=2, group=WPO, parameterName=WPO-Noise_Tiling_X}
    B=ScalarParameter() {defaultValue=2, group=WPO, parameterName=WPO-Noise_Tiling_Y}
   )
  )
  n71 = AppendVector(
   A=ScalarParameter() {defaultValue=0.2, group=WPO, parameterName=WPO-PanSpeed_X}
   B=ScalarParameter() {defaultValue=0.3, group=WPO, parameterName=WPO-PanSpeed_Y}
  )
  n72 = Multiply(
   A=TextureSampleParameter2D(UVs=Panner(Coordinate=@n70, Speed=@n71)) {group=WPO, parameterName=WPO_Noise, sortPriority=0, texture=/Game/Global/Textures/Default/Noise/T_Noise_Cloud}.B
   B=ScalarParameter() {defaultValue=1, group=WPO, parameterName=WPO_Intensity, sortPriority=2}
  )
  n73 = Power(Base=ComponentMask(None=@n66) {g=true}, Exp=ScalarParameter() {defaultValue=2, group=WPO, parameterName=WPO-VerticalGradMask_Power})
  n74 = ScalarParameter() {defaultValue=1, group=WPO, parameterName=WPO-VerticalGradMask_Mult}
  n75 = Multiply(A=TextureCoordinate(), B=AppendVector(A=Constant(), B=ScalarParameter() {defaultValue=5, group=WPO, parameterName=WPO-Tiling_Grad}))
  n76 = AppendVector(A=Constant(), B=ScalarParameter() {group=WPO, parameterName=WPO-Offset_Grad})
  n77 = Multiply(A=Clamp(None=Multiply(A=@n73, B=@n74)), B=ComponentMask(None=Add(A=@n75, B=@n76)) {g=true})
  Multiply(
   A=TextureSampleParameter2D(UVs=TextureCoordinate() {coordinateIndex=1}) {parameterName=WPO_MaskTexture, texture=&/T_ElasticCosmic_Pickaxe_WPOMask}.RGB
   B=Multiply(A=Multiply(A=@n69, B=@n72), B=Clamp(None=@n77))
  )
  n78 = Add(A=@n54, B=@n63)
  n79 = Constant() {r=1}
  n80 = ScalarParameter() {defaultValue=0.2, group=Mask, parameterName=Mask_Value}
  n81 = Constant()
  Reroute(None=LinearInterpolate(
    A=@n64
    B=If(A=@n78, B=@n80, "A > B"=@n79, "A == B"=@n79, "A < B"=@n81)
    Alpha=ScalarParameter() {defaultValue=1, group=Mask, parameterName=Mask_Is_HardEdge?, sliderMax=1}
   ))
  VectorParameter() {group=Color, parameterName=Primary Color, defaultValue=<rgba 0 0.05 1 1>}
  VectorParameter() {group=Color, parameterName=Secondary Color, defaultValue=<rgba 0.191 0 0.5729 1>}
 ]
}
