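# Translucent, unlit, two-sided Niagara sprite and mesh-particle material for spike effects on the Cosmic Elastic pickaxe.
# A spike texture, rotated by a dynamic parameter and distorted by panning UV1 noise, supplies shape and alpha.
# Emissive multiplies its green channel by secondary color, a U gradient, a V-indexed curve-atlas green value and particle color.
# Opacity combines texture and particle alpha, with an optional V-gradient mask blended in by parameter.
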Material /BRCosmetics/Weapons/FORT_Melee/Pickaxe_Elastic/Cosmic/M_Pickaxe_Elastic_Spike_01
graph {
 properties {
  bUsedWithNiagaraMeshParticles=true
  bUsedWithNiagaraSprites=true
  blendMode=BLEND_Translucent
  pixelDepthOffsetMode=PDOM_Legacy
  shadingModel=MSM_Unlit
  twoSided=true
  usedShadingModels=Unlit
 }
 nodes [
  n1 = Multiply(
   A=ScalarParameter() {defaultValue=1, group=Color, parameterName=Secondary Color Intensity}
   B=VectorParameter() {group=Color, parameterName=Secondary Color, defaultValue=<rgba 0.191 0 0.5729 1>}.RGB
  )
  n2 = TextureCoordinate()
  n3 = Multiply(
   A=Power(Base=ComponentMask(None=@n2) {r=true}, Exp=ScalarParameter() {defaultValue=0.5, parameterName=Power})
   B=ScalarParameter() {defaultValue=2, parameterName=Intensity}
  )
  n4 = CurveAtlasRowParameter(CurveTime=ComponentMask(None=TextureCoordinate()) {g=true}) {
   atlas=/Game/Athena/LinearColorCurves/TempAthenaCurveAtlas
   curve=/Game/Weapons/FORT_Melee/Pickaxe_Elastic/Cosmic/Curve_Pickaxe_Elastic_Cosmic_01
   defaultValue=12
  }
  n5 = TextureCoordinate()
  n6 = Add(
   A=ScalarParameter() {parameterName=TextuteRotation}
   B=DynamicParameter() {defaultValue=<rgba 0 1 1 1>, paramNames [TextuteRotation, Param2, Param3, Param4]}.TextuteRotation
  )
  n7 = TextureCoordinate() {coordinateIndex=1}
  n8 = AppendVector(
   A=ScalarParameter() {defaultValue=2, group=Distortion, parameterName=Distortion_Tile_U}
   B=ScalarParameter() {defaultValue=0.5, group=Distortion, parameterName=Distortion_Tile_V}
  )
  n9 = ScalarParameter() {defaultValue=-0.5, group=Distortion, parameterName=Distortion_Pan_U}
  n10 = ScalarParameter() {defaultValue=0.3, group=Distortion, parameterName=Distortion_Pan_V}
  n11 = TextureSample(UVs=Panner(Coordinate=Multiply(A=@n7, B=@n8), Speed=AppendVector(A=@n9, B=@n10)) {speedY=0.3}) {texture=&T_NoisePack_01}
  n12 = ScalarParameter() {defaultValue=0.1, group=Distortion, parameterName=Distortion_Intensity}
  n13 = TextureSample(UVs=Add(A=Rotator(Coordinate=@n5, Time=@n6), B=Multiply(A=@n11.R, B=@n12) {constB=0.1})) {texture=&/T_Pickaxe_Elastic_Spike_02}
  n14 = ParticleColor()
  n15 = Multiply(A=Multiply(A=Multiply(A=@n1, B=@n3), B=Multiply(A=@n4.G, B=@n13.G)), B=@n14.RGB)
  MaterialFunctionCall(Emissive=@n15) {materialFunction=&MF_FX_EmissiveAdjustments}
  Reroute(None=@n15) {desc=Emissive}
  n16 = Multiply(A=@n14.A, B=Multiply(A=@n13.A, B=ScalarParameter() {defaultValue=1, parameterName=Alpha}))
  n17 = ComponentMask(None=TextureCoordinate()) {g=true}
  n18 = ScalarParameter() {defaultValue=0.5, parameterName=MaskPower}
  n19 = Saturate(None=Multiply(A=Power(Base=@n17, Exp=@n18), B=ScalarParameter() {defaultValue=1, parameterName=MaskIntensity}))
  Reroute(None=LinearInterpolate(A=@n16, B=Multiply(A=@n16, B=@n19), Alpha=ScalarParameter() {parameterName=UseOpacityMask})) {desc=Opacity}
 ]
}
