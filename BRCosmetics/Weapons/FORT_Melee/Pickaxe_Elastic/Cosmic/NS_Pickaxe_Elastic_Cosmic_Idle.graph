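# Niagara idle effect for the Cosmic Elastic pickaxe cosmetic, drawing a glowing blade with streaks, smoke, sparks and a light.
# A sword mesh, blade airglow, streaks and lingering sparks follow the User.SkeletalMesh, while smoke, other sparks, wrist glow and light use system-relative offsets.
# Primary and Secondary Color user parameters tint everything, and two spawn groups mirror the streak motion.
# The decal glow and older L/R sprite emitters are disabled.
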NiagaraSystem /BRCosmetics/Weapons/FORT_Melee/Pickaxe_Elastic/Cosmic/NS_Pickaxe_Elastic_Cosmic_Idle
dependencies {
 usedDataInterfaces [&NiagaraDataInterfaceCurve, &NiagaraDataInterfaceVector2DCurve]
 usedDynamicInputs []
 usedModules [&SystemState]
}
emitters {
 Blade_Airglow {
  emitterUpdateScript [
   &EmitterState(Inactive Response=::NewEnumerator1, Life Cycle Mode=::NewEnumerator1, Loop Duration=0.25, UseLoopDelay=false)
   &SpawnBurst_Instantaneous(Spawn Count=1, Use Spawn Probability=false)
  ]
  particleSpawnScript [
   /Niagara/Modules/Spawn/Initialization/InitializeParticle(
    Color=$User.Secondary Color
    Lifetime=0.25
    Mesh Scale=<200 30 1>
    Module.Write Color=true
    Module.Write Lifetime=true
    Module.Write Mass=true
    Module.Write Material Random=true
    Module.Write Position=false
    Module.Write Scale=true
    Module.Write SpriteRotation=false
    Module.Write SpriteSize=true
    Position=&SimulationPosition()
   )
   &DynamicMaterialParameters(
    Index 0 Param 1=1
    Module.Param0WriteEnabled=true
    Module.Param1WriteEnabled=true
    Module.Param2WriteEnabled=true
    Module.Param3WriteEnabled=true
   )
   /Niagara/Modules/Spawn/Orientation/V2/InitialMeshOrientation(Orientation Axis:Vector3f=<-1 0 0>, Rotation=<0.5 0 0>, Use Orientation Vector=true, Use Rotation Vector=true)
  ]
  particleUpdateScript [
   &ParticleState()
   &SkeletalMeshLocation(
    Bone / Socket Index=2
    Inherit Velocity=false
    Interpolate Mesh Sampling=false
    MaskSpawnGroup=false
    Mesh Sampling Type=::NewEnumerator5
    Skeletal Mesh=$User.SkeletalMesh
    Socket Sampling Mode=::NewEnumerator5
   )
   &ScaleColor(Scale Alpha=0.6, Scale RGB=&VectorFromFloat(Value=0.1), ScaleA=true, ScaleRGB=true)
   &CameraOffset(Camera Offset Amount=20)
  ]
  rendererClasses [&NiagaraMeshRendererProperties]
  renderers [{rendererClass=&NiagaraMeshRendererProperties, rendererIndex=0}]
 }
 Bottom_Streaks {
  emitterUpdateScript [
   &EmitterState(Life Cycle Mode=::NewEnumerator1, Loop Duration=1, UseLoopDelay=false)
   &SpawnRate(SpawnRate=8, Use Spawn Probability=false)
  ]
  particleSpawnScript [
   &InitializeParticle(
    Color=$User.Secondary Color
    Mesh Scale=&UniformRangedVector(Minimum=<5 15 0>, Maximum=<10 28 1>)
    Module.Write Color=true
    Module.Write Lifetime=true
    Module.Write Mass=true
    Module.Write Material Random=true
    Module.Write Position=true
    Module.Write Scale=true
    Module.Write SpriteRotation=true
    Module.Write SpriteSize=true
    Sprite Size=&UniformRangedVector2D(Minimum=<5 15>, Maximum=<10 28>)
   )
   &FN_SetSpawnGroup_v2(Number of Spawn Groups=2)
   /SetVariables(Particles.SubImageIndex=&UniformRangedFloat(Minimum=0, Maximum=1, Randomness Mode=::NewEnumerator1)) {isSetParameters=true, module=SetVariables_#1}
   /SetVariables(Particles.SpriteFacing=&TransformVector(Vector=<0 1 0>, Destination Space=::World)) {enabled=false, isSetParameters=true, module=SetVariables_#2}
   &SpriteFacingAndAlignment(Sprite Alignment=<-1 1 0>, Sprite Facing=<0 1 0>, WriteAlignment=true, WriteFacing=true)
  ]
  particleUpdateScript [
   &ParticleState()
   /SetVariables(Particles.SpriteRotation:NiagaraFloat=&MaskFloatBySpawnGroup(
     Masked Float=&FloatFromCurve(FloatCurve=DataInterface() {propertyValues {Curve {keys [<at 0 0>,<at 1 -15>]}}})
     Passthrough Float=&FloatFromCurve(FloatCurve=DataInterface() {propertyValues {Curve {keys [<at 0 0>,<at 1 15>]}}})
    )) {enabled=false, isSetParameters=true, module=SetVariables_#3}
   &SkeletalMeshLocation(
    Bone / Socket Index=&SpawnGroup()
    Inherit Velocity=false
    MaskSpawnGroup=false
    Mesh Sampling Type=::NewEnumerator5
    Sampled Position Offset=&MaskVectorBySpawnGroup(
     Masked Vector=&Lerp_Vector(A=<0 0 3>, B=<45 0 10>, Alpha=$Particles.NormalizedAge)
     Passthrough Vector=&Lerp_Vector(A=<0 0 -3>, B=<45 0 -10>, Alpha=$Particles.NormalizedAge)
    )
    Sampled Position Offset Coordinate Space:ENiagaraMeshTransforms=::NewEnumerator3
    Skeletal Mesh=$User.SkeletalMesh
    Socket Sampling Mode=::NewEnumerator5
   )
   /SetVariables(Particles.Previous.Velocity:Vector3f=<-1 0 0>) {isSetParameters=true, module=SetVariables_#4}
   /SetVariables(Particles.StreaksAlignment:Vector3f=&TransformVector(Vector=<-1 0 0>, Destination Space=::World)) {enabled=false, isSetParameters=true}
   &DynamicMaterialParameters(
    Index 0 Param 1=&MaskFloatBySpawnGroup(
     Masked Float=&Lerp_Float(A=-1, B=0, Alpha=$Particles.NormalizedAge)
     Passthrough Float=&Lerp_Float(A=1, B=0, Alpha=$Particles.NormalizedAge)
    )
    Module.Param0WriteEnabled=true
    Module.Param1WriteEnabled=true
    Module.Param2WriteEnabled=true
    Module.Param3WriteEnabled=true
   )
   &CameraOffset(Camera Offset Amount=-2)
   &ScaleColor(
    Scale Alpha=&FloatFromCurve(FloatCurve=DataInterface() {propertyValues {Curve {keys [<at 0 0 RCIM_Cubic>,<at 0.2 1 RCIM_Cubic>,<at 1 0 RCIM_Cubic>]}}})
    Scale RGB=&MakeVectorFromLinearColorRGB(Color=<rgba 1 1 1 1>)
    ScaleA=true
    ScaleRGB=true
   )
   &ScaleSpriteSize(Scale Factor=&Vector2DFromFloat(Value=&FloatFromCurve(FloatCurve=DataInterface() {propertyValues {Curve {keys [<at 0 0>,<at 0.2 1.5>]}}})))
   &SolveForcesAndVelocity(Clamp Velocity=false, Limit Acceleration=false)
  ]
  rendererClasses [&NiagaraSpriteRendererProperties]
  renderers [{rendererClass=&NiagaraSpriteRendererProperties}]
 }
 Decal_Glow {
  bEnabled=false
  emitterUpdateScript [
   &EmitterState()
   &SpawnBurst_Instantaneous(Spawn Count=1, Use Spawn Probability=false)
   &SpawnRate(SpawnRate=3, Use Spawn Probability=false) {enabled=false}
  ]
  particleSpawnScript [
   /Game/Effects/Niagara/Modules/Partition/FN_ParticlePartition_v2/FN_ParticlePartition_v2()
   &InitializeParticle(
    Color=$User.Primary Color
    Module.Write Color=true
    Module.Write Lifetime=true
    Module.Write Mass=true
    Module.Write Material Random=true
    Module.Write Position=false
    Module.Write Scale=false
    Module.Write SpriteRotation=false
    Module.Write SpriteSize=true
    Position=&SimulationPosition()
    Sprite Size=&Vector2DFromFloat(Value=25)
   )
   /Game/Effects/Niagara/Modules/BoneFlow/FN_Initialize_BoneFlow_Attributes_v2()
   &SystemLocation(MaskSpawnGroup=false, Offset=<10 0 0>)
  ]
  particleUpdateScript [
   &ParticleState()
   &DynamicMaterialParameters(
    Index 0 Param 1=1.5
    Module.Param0WriteEnabled=true
    Module.Param1WriteEnabled=false
    Module.Param2WriteEnabled=false
    Module.Param3WriteEnabled=false
   ) {module=DynamicMaterialParameters001}
   &ScaleColor(
    Scale Alpha=&FloatFromCurve(FloatCurve=DataInterface() {propertyValues {Curve {keys [<at 0.2 1>,<at 0.8 1>]}}})
    Scale RGB=&VectorFromFloat(Value=2)
    ScaleA=true
    ScaleRGB=true
   )
   &CameraOffset(Camera Offset Amount=20)
   &SolveForcesAndVelocity(Clamp Velocity=false, Limit Acceleration=false)
  ]
  rendererClasses [&NiagaraSpriteRendererProperties]
  renderers [{rendererClass=&NiagaraSpriteRendererProperties}]
 }
 Decal_Smoke {
  emitterUpdateScript [
   &EmitterState()
   &SpawnRate(SpawnRate=&UniformRangedFloat(Minimum=2.5, Maximum=3), Use Spawn Probability=false)
  ]
  particleSpawnScript [
   &FN_ParticlePartition_v2()
   &InitializeParticle(
    Color=$User.Primary Color
    Lifetime=&UniformRangedFloat(Minimum=2.2, Maximum=1.5)
    Module.Write Color=true
    Module.Write Lifetime=true
    Module.Write Mass=true
    Module.Write Material Random=true
    Module.Write Position=false
    Module.Write Scale=false
    Module.Write SpriteRotation=true
    Module.Write SpriteSize=true
    Position=&SimulationPosition()
    Sprite Rotation=&UniformRangedFloat(Minimum=180, Maximum=-180)
    Sprite Size=&Vector2DFromFloat(Value=&UniformRangedFloat(Minimum=18, Maximum=22))
   )
   &FN_Initialize_BoneFlow_Attributes_v2()
   &SystemLocation(MaskSpawnGroup=false, Offset=<20 0 0>)
   &AddVelocity(Velocity=&UniformRangedVector(Minimum=<-10 0 0>, Maximum=<-6 0 0>))
  ]
  particleUpdateScript [
   &ParticleState()
   &ScaleColor(
    Scale Alpha=&FloatFromCurve(FloatCurve=DataInterface() {propertyValues {Curve {keys [
         {interpMode=RCIM_Cubic, tangentMode=RCTM_Break, time=0, value=0}
         {interpMode=RCIM_Cubic, tangentMode=RCTM_User, time=0.3024, value=1.1314}
         {interpMode=RCIM_Cubic, tangentMode=RCTM_Break, time=1, value=0}
        ]}}})
    Scale RGB=&VectorFromFloat(Value=&UniformRangedFloat(Minimum=10, Maximum=15))
    ScaleA=true
    ScaleRGB=true
   )
   &CameraOffset(Camera Offset Amount=20)
   &SolveForcesAndVelocity(Clamp Velocity=false, Limit Acceleration=false)
   &DynamicMaterialParameters(
    Index 0 Param 1=&FloatFromCurve(FloatCurve=DataInterface() {propertyValues {Curve {keys [{interpMode=RCIM_Cubic, leaveTangent=2.3044, tangentMode=RCTM_Break, time=0, value=0.25}, <at 1 1>]}}})
    Index 0 Param 2=0.5
    Module.Param0WriteEnabled=true
    Module.Param1WriteEnabled=true
    Module.Param2WriteEnabled=true
    Module.Param3WriteEnabled=true
   )
  ]
  rendererClasses [&NiagaraSpriteRendererProperties]
  renderers [{rendererClass=&NiagaraSpriteRendererProperties}]
 }
 Decal_Smoke001 {
  emitterUpdateScript [
   &EmitterState()
   &SpawnRate(SpawnRate=5, Use Spawn Probability=false)
  ]
  particleSpawnScript [
   &FN_ParticlePartition_v2()
   &InitializeParticle(
    Color=$User.Primary Color
    Lifetime=&UniformRangedFloat(Minimum=1.5, Maximum=2)
    Module.Write Color=true
    Module.Write Lifetime=true
    Module.Write Mass=true
    Module.Write Material Random=true
    Module.Write Position=false
    Module.Write Scale=false
    Module.Write SpriteRotation=true
    Module.Write SpriteSize=true
    Position=&SimulationPosition()
    Sprite Rotation=&UniformRangedFloat(Minimum=-360, Maximum=360)
    Sprite Size=&Vector2DFromFloat(Value=&UniformRangedFloat(Minimum=10, Maximum=30))
   )
   &FN_Initialize_BoneFlow_Attributes_v2()
   &SystemLocation(MaskSpawnGroup=false, Offset=&UniformRangedVector(Minimum=<3 0 0>, Maximum=<6 0 0>))
   &AddVelocity(Velocity=&UniformRangedVector(Minimum=<-20 0 0>, Maximum=<-25 0 0>))
  ]
  particleUpdateScript [
   &ParticleState()
   &ScaleColor(
    Scale Alpha=&FloatFromCurve(FloatCurve=DataInterface() {propertyValues {Curve {keys [
         {interpMode=RCIM_Cubic, tangentMode=RCTM_Break, time=0, value=0}
         {arriveTangent=0.4148, interpMode=RCIM_Cubic, leaveTangent=0.4148, time=0.2646, value=0.9748}
         {arriveTangent=-1.2741, interpMode=RCIM_Cubic, leaveTangent=-1.2741, tangentMode=RCTM_User, time=0.5601, value=0.2323}
         <at 1 0 RCIM_Cubic>
        ]}}})
    Scale RGB=&VectorFromFloat(Value=&UniformRangedFloat(Minimum=22, Maximum=33))
    ScaleA=true
    ScaleRGB=true
   )
   &CameraOffset(Camera Offset Amount=20)
   &Drag(Drag=0.5, Use Linear Drag=true, Use Rotational Drag=false)
   &SolveForcesAndVelocity(Clamp Velocity=false, Limit Acceleration=false)
   &DynamicMaterialParameters(
    Index 0 Param 1=&FloatFromCurve(FloatCurve=DataInterface() {propertyValues {Curve {keys [{interpMode=RCIM_Cubic, leaveTangent=2.3044, tangentMode=RCTM_Break, time=0, value=0.25}, <at 1 1>]}}})
    Index 0 Param 2=0.5
    Module.Param0WriteEnabled=true
    Module.Param1WriteEnabled=true
    Module.Param2WriteEnabled=true
    Module.Param3WriteEnabled=true
   )
   &ScaleSpriteSize(Scale Factor=&Vector2DFromFloat(Value=&FloatFromCurve(FloatCurve=DataInterface() {propertyValues {Curve {keys [<at 0.1461 0.9473 RCIM_Cubic>,<at 1 0.3698>]}}})))
  ]
  rendererClasses [&NiagaraSpriteRendererProperties]
  renderers [{rendererClass=&NiagaraSpriteRendererProperties}]
 }
 Light {
  emitterUpdateScript [
   &EmitterState(Inactive Response=::NewEnumerator1, Life Cycle Mode=::NewEnumerator1, Loop Duration=1, UseLoopDelay=false)
   &SpawnBurst_Instantaneous(Spawn Count=1, Use Spawn Probability=false)
  ]
  particleSpawnScript [
   &InitializeParticle(
    Color=$User.Primary Color
    Module.Write Color=true
    Module.Write Lifetime=true
    Module.Write Mass=true
    Module.Write Material Random=true
    Module.Write Position=false
    Module.Write Scale=false
    Module.Write SpriteRotation=false
    Module.Write SpriteSize=true
    Position=&SimulationPosition()
    Sprite Size=&Vector2DFromFloat(Value=150)
   )
   &DynamicMaterialParameters(
    Index 0 Param 1=1
    Module.Param0WriteEnabled=true
    Module.Param1WriteEnabled=true
    Module.Param2WriteEnabled=true
    Module.Param3WriteEnabled=true
   )
   &SystemLocation(MaskSpawnGroup=false, Offset=<2 -2 -2>)
  ]
  particleUpdateScript [
   &CameraOffset(Camera Offset Amount=15)
   &ParticleState()
   &ScaleColor(Scale Alpha=0.1, Scale RGB=&VectorFromFloat(Value=0.1), ScaleA=true, ScaleRGB=true)
   &FN_Light_v2(Light Exponent=3, Light Radius=20, LightEnabled=true, Position Offset:Vector3f=<5 -5 -5>, Scale Color=200)
  ]
  rendererClasses [&NiagaraLightRendererProperties]
  renderers [{rendererClass=&NiagaraLightRendererProperties, rendererIndex=0}]
 }
 Lingering_Sparks {
  emitterUpdateScript [
   &EmitterState()
   &SpawnRate(SpawnRate=&UniformRangedFloat(Minimum=3, Maximum=4), Use Spawn Probability=false)
  ]
  particleSpawnScript [
   &FN_ParticlePartition_v2()
   &InitializeParticle(
    Color=$User.Primary Color
    Lifetime=&UniformRangedFloat(Minimum=0.3, Maximum=0.4)
    Module.Write Color=true
    Module.Write Lifetime=true
    Module.Write Mass=true
    Module.Write Material Random=true
    Module.Write Position=false
    Module.Write Scale=false
    Module.Write SpriteRotation=true
    Module.Write SpriteSize=true
    Position=&SimulationPosition()
    Sprite Rotation=&UniformRangedFloat(Minimum=180, Maximum=-180)
    Sprite Size=&Vector2DFromFloat(Value=&UniformRangedFloat(Minimum=3, Maximum=1))
   )
   &FN_Initialize_BoneFlow_Attributes_v2()
   &SystemLocation(MaskSpawnGroup=false, Offset=&UniformRangedVector(Minimum=<0 0 0>, Maximum=<1 1 1>))
   &SphereLocation(
    MaskSpawnGroup=false
    Override Seed=false
    Sphere Radius=6
    Surface Only=false
    UseNonUniformScale=false
    UseOffset=false
    UseSphereOrientation=false
   )
   &SkeletalMeshLocation(
    Bone / Socket Index=&SpawnGroup()
    Inherit Velocity=false
    MaskSpawnGroup=false
    Mesh Sampling Type=::NewEnumerator5
    Sampled Position Offset=&UniformRangedVector(Minimum=<0 0 0>, Maximum=<20 0 0>)
    Sampled Position Offset Coordinate Space:ENiagaraMeshTransforms=::NewEnumerator3
    Skeletal Mesh=$User.SkeletalMesh
    Socket Sampling Mode=::NewEnumerator5
   )
  ]
  particleUpdateScript [
   &ParticleState()
   &ScaleColor(
    Scale Alpha=&FloatFromCurve(FloatCurve=DataInterface() {propertyValues {Curve {keys [
         {interpMode=RCIM_Cubic, tangentMode=RCTM_Break, time=0, value=0}
         <at 0.4 1>
         {interpMode=RCIM_Cubic, tangentMode=RCTM_User, time=0.65, value=1}
         {interpMode=RCIM_Cubic, tangentMode=RCTM_Break, time=1, value=0}
        ]}}})
    Scale RGB=&VectorFromFloat(Value=&UniformRangedFloat(Minimum=10, Maximum=20))
    ScaleA=true
    ScaleRGB=true
   )
   &CameraOffset(Camera Offset Amount=20)
   &VectorNoiseForce(Force Amount=30, Module.Use Random Noise Time Interval=false)
   &GravityForce(Gravity=<0 0 -100>)
   &Drag(Use Linear Drag=true, Use Rotational Drag=false)
   &SolveForcesAndVelocity(Clamp Velocity=false, Limit Acceleration=false)
  ]
  rendererClasses [&NiagaraSpriteRendererProperties]
  renderers [{rendererClass=&NiagaraSpriteRendererProperties}]
 }
 Mesh_Sword {
  emitterUpdateScript [
   &EmitterState(
    Inactive Response=::NewEnumerator1
    Life Cycle Mode=::NewEnumerator1
    Loop Behavior=::NewEnumerator1
    Loop Duration Mode=::NewEnumerator1
    UseLoopDelay=false
   )
   &SpawnBurst_Instantaneous(Spawn Count=1, Use Spawn Probability=false)
  ]
  particleSpawnScript [&InitializeParticle(
    Lifetime=5
    Module.Write Color=true
    Module.Write Lifetime=true
    Module.Write Mass=true
    Module.Write Material Random=true
    Module.Write Position=false
    Module.Write Scale=false
    Module.Write SpriteRotation=false
    Module.Write SpriteSize=false
    Position=&SimulationPosition()
    Sprite Size=<5 5>
   )]
  particleUpdateScript [
   &ParticleState(Kill Particles When Lifetime Has Elapsed=false)
   &SkeletalMeshLocation(Inherit Velocity=false, MaskSpawnGroup=false, Module.Override Randomness=false, Skeletal Mesh=$User.SkeletalMesh)
   &self:DynParam_ColorPicker(Primary Color:LinearColor=$User.Primary Color, Secondary Color:LinearColor=$User.Secondary Color) {module=DynParam_ColorPicker}
  ]
  rendererClasses [&NiagaraMeshRendererProperties]
  renderers [{rendererClass=&NiagaraMeshRendererProperties, rendererIndex=0}]
 }
 Middle_Streaks {
  emitterUpdateScript [
   &EmitterState(Life Cycle Mode=::NewEnumerator1, Loop Duration=1, UseLoopDelay=false)
   &SpawnRate(SpawnRate=8, Use Spawn Probability=false)
  ]
  particleSpawnScript [
   &InitializeParticle(
    Color=$User.Secondary Color
    Lifetime=0.5
    Module.Write Color=true
    Module.Write Lifetime=true
    Module.Write Mass=true
    Module.Write Material Random=true
    Module.Write Position=true
    Module.Write Scale=false
    Module.Write SpriteRotation=true
    Module.Write SpriteSize=true
    Sprite Rotation=-5
    Sprite Size=&UniformRangedVector2D(Minimum=<10 25>, Maximum=<12 35>)
   )
   &FN_SetSpawnGroup_v2(Number of Spawn Groups=2)
   /SetVariables(Particles.SubImageIndex=&UniformRangedFloat(Minimum=0, Maximum=2, Randomness Mode=::NewEnumerator1)) {isSetParameters=true, module=SetVariables_#5}
   /SetVariables(Particles.SpriteRotation:NiagaraFloat=&MaskFloatBySpawnGroup(Masked Float=-75, Passthrough Float=75)) {isSetParameters=true}
   &SpriteFacingAndAlignment(Sprite Alignment=<-1 1 0>, Sprite Facing=<0 1 0>, WriteAlignment=true, WriteFacing=true)
  ]
  particleUpdateScript [
   &ParticleState()
   &SkeletalMeshLocation(
    Bone / Socket Index=&SpawnGroup()
    Inherit Velocity=false
    MaskSpawnGroup=false
    Mesh Sampling Type=::NewEnumerator5
    Sampled Position Offset=&MaskVectorBySpawnGroup(Masked Vector=<-2 0 0>, Passthrough Vector=<-2 0 0>)
    Sampled Position Offset Coordinate Space:ENiagaraMeshTransforms=::NewEnumerator3
    Skeletal Mesh=$User.SkeletalMesh
    Socket Sampling Mode=::NewEnumerator5
   )
   &AddVelocity(Velocity=&MaskVectorBySpawnGroup(Masked Vector=<-10 0 -15>, Passthrough Vector=<-10 0 15>))
   &SpriteRotationRate(Rotation Rate=&MaskFloatBySpawnGroup(Masked Float=150, Passthrough Float=-150))
   /SetVariables(Particles.SpriteFacing=&TransformVector(Vector=<0 1 0>, Destination Space=::World)) {enabled=false, isSetParameters=true, module=SetVariables_#2}
   /SetVariables(Particles.StreaksAlignment:Vector3f=&TransformVector(Vector=<-1 0 -0.05>, Destination Space=::World)) {enabled=false, isSetParameters=true}
   /SetVariables(Particles.Previous.Velocity:Vector3f=<-1 0 -0.05>) {isSetParameters=true, module=SetVariables_#6}
   &DynamicMaterialParameters(Module.Param0WriteEnabled=true, Module.Param1WriteEnabled=true, Module.Param2WriteEnabled=true, Module.Param3WriteEnabled=true)
   &CameraOffset(Camera Offset Amount=-2.5)
   &ScaleColor(
    Scale Alpha=&FloatFromCurve(FloatCurve=DataInterface() {propertyValues {Curve {keys [<at 0 0>,<at 0.2 1>,<at 1 0>]}}})
    Scale RGB=&MakeVectorFromLinearColorRGB(Color=<rgba 1 1 1 1>)
    ScaleA=true
    ScaleRGB=true
   )
   &ScaleSpriteSize(Scale Factor=&Vector2DFromCurve(Vector2Curve=DataInterface() {propertyValues {XCurve {keys [<at 0 1>,<at 1 1>]}, YCurve {keys [<at 0 0>,<at 1 0.8>]}}}, ScaleCurve=<1 1>))
   &SolveForcesAndVelocity(Clamp Velocity=false, Limit Acceleration=false)
  ]
  rendererClasses [&NiagaraSpriteRendererProperties]
  renderers [{rendererClass=&NiagaraSpriteRendererProperties}]
 }
 Particle_Spark {
  emitterUpdateScript [
   &EmitterState()
   &SpawnRate(SpawnRate=&UniformRangedFloat(Minimum=6, Maximum=8), Use Spawn Probability=false)
  ]
  particleSpawnScript [
   &FN_ParticlePartition_v2()
   &InitializeParticle(
    Color=$User.Primary Color
    Lifetime=&UniformRangedFloat(Minimum=1, Maximum=1.5)
    Module.Write Color=true
    Module.Write Lifetime=true
    Module.Write Mass=true
    Module.Write Material Random=true
    Module.Write Position=false
    Module.Write Scale=false
    Module.Write SpriteRotation=true
    Module.Write SpriteSize=true
    Position=&SimulationPosition()
    Sprite Rotation=&UniformRangedFloat(Minimum=180, Maximum=-180)
    Sprite Size=&Vector2DFromFloat(Value=&UniformRangedFloat(Minimum=3, Maximum=1))
   )
   &FN_Initialize_BoneFlow_Attributes_v2()
   &SystemLocation(MaskSpawnGroup=false, Offset=<20 0 0>)
   &SphereLocation(
    MaskSpawnGroup=false
    Override Seed=false
    Sphere Radius=6
    Surface Only=false
    UseNonUniformScale=false
    UseOffset=false
    UseSphereOrientation=false
   )
   &AddVelocity(Velocity=&UniformRangedVector(Minimum=<-15 0 0>, Maximum=<-80 0 0>))
  ]
  particleUpdateScript [
   &ParticleState()
   &ScaleColor(
    Scale Alpha=&FloatFromCurve(FloatCurve=DataInterface() {propertyValues {Curve {keys [
         {interpMode=RCIM_Cubic, tangentMode=RCTM_Break, time=0, value=0}
         <at 0.4 1>
         {interpMode=RCIM_Cubic, tangentMode=RCTM_User, time=0.65, value=1}
         {interpMode=RCIM_Cubic, tangentMode=RCTM_Break, time=1, value=0}
        ]}}})
    Scale RGB=&VectorFromFloat(Value=&UniformRangedFloat(Minimum=40, Maximum=150))
    ScaleA=true
    ScaleRGB=true
   )
   &CameraOffset(Camera Offset Amount=20)
   &VectorNoiseForce(Force Amount=450, Module.Use Random Noise Time Interval=true, Random Noise Time Interval=0.6)
   &Drag(Use Linear Drag=true, Use Rotational Drag=false)
   &SolveForcesAndVelocity(Clamp Velocity=false, Limit Acceleration=false)
  ]
  rendererClasses [&NiagaraSpriteRendererProperties]
  renderers [{rendererClass=&NiagaraSpriteRendererProperties}]
 }
 Particle_Spark001 {
  emitterUpdateScript [
   &EmitterState()
   &SpawnRate(SpawnRate=&UniformRangedFloat(Minimum=8, Maximum=12), Use Spawn Probability=false)
  ]
  particleSpawnScript [
   &FN_ParticlePartition_v2()
   &InitializeParticle(
    Color=$User.Primary Color
    Lifetime=&UniformRangedFloat(Minimum=0.9, Maximum=1.1)
    Module.Write Color=true
    Module.Write Lifetime=true
    Module.Write Mass=true
    Module.Write Material Random=true
    Module.Write Position=false
    Module.Write Scale=false
    Module.Write SpriteRotation=true
    Module.Write SpriteSize=true
    Position=&SimulationPosition()
    Sprite Rotation=&UniformRangedFloat(Minimum=180, Maximum=-180)
    Sprite Size=&Vector2DFromFloat(Value=&UniformRangedFloat(Minimum=2, Maximum=3))
   )
   &FN_Initialize_BoneFlow_Attributes_v2()
   &SystemLocation(MaskSpawnGroup=false)
   &SphereLocation(
    MaskSpawnGroup=false
    Offset=<12 0 0>
    Override Seed=false
    Sphere Radius=&UniformRangedFloat(Minimum=2, Maximum=3)
    Surface Only=false
    UseNonUniformScale=false
    UseOffset=true
    UseSphereOrientation=false
   )
   &AddVelocity(Velocity=&UniformRangedVector(Minimum=<-60 0 0>, Maximum=<-150 0 0>))
  ]
  particleUpdateScript [
   &ParticleState()
   &ScaleColor(
    Scale Alpha=&FloatFromCurve(FloatCurve=DataInterface() {propertyValues {Curve {keys [
         {interpMode=RCIM_Cubic, tangentMode=RCTM_Break, time=0, value=0}
         <at 0.4 1>
         {interpMode=RCIM_Cubic, tangentMode=RCTM_User, time=0.65, value=1}
         {interpMode=RCIM_Cubic, tangentMode=RCTM_Break, time=1, value=0}
        ]}}})
    Scale RGB=&VectorFromFloat(Value=&UniformRangedFloat(Minimum=10, Maximum=20))
    ScaleA=true
    ScaleRGB=true
   )
   &CameraOffset(Camera Offset Amount=20)
   &VectorNoiseForce(Force Amount=500, Module.Use Random Noise Time Interval=true)
   /Niagara/Modules/Update/Forces/CurlNoiseForce(Noise Frequency=0.1, Noise Quality / Cost=::NewEnumerator0, Noise Strength=30, Pan Noise=false)
   &Drag(Drag=1.95, Use Linear Drag=true, Use Rotational Drag=false)
   &SolveForcesAndVelocity(Clamp Velocity=false, Limit Acceleration=false)
  ]
  rendererClasses [&NiagaraSpriteRendererProperties]
  renderers [{rendererClass=&NiagaraSpriteRendererProperties}]
 }
 Sprite_L_01001 {
  bEnabled=false
  emitterUpdateScript [
   &EmitterState()
   &SpawnRate(SpawnRate=7, Use Spawn Probability=false)
  ]
  particleSpawnScript [
   &InitializeParticle(
    Color=$User.Secondary Color
    Lifetime=0.4
    Module.Write Color=true
    Module.Write Lifetime=true
    Module.Write Mass=true
    Module.Write Material Random=true
    Module.Write Position=true
    Module.Write Scale=false
    Module.Write SpriteRotation=true
    Module.Write SpriteSize=true
    Sprite Size=&UniformRangedVector2D(Minimum=<8.5 23>, Maximum=<10 28>)
   )
   &SystemLocation(MaskSpawnGroup=false, Offset=<-12 -1.5 4.8>)
   &AddVelocity(Velocity=<-100 0 -20>)
   /SetVariables(Particles.SubImageIndex=&UniformRangedFloat(Minimum=0, Maximum=2, Randomness Mode=::NewEnumerator1)) {isSetParameters=true, module=SetVariables_#1}
  ]
  particleUpdateScript [
   &ParticleState()
   /SetVariables(Particles.Previous.Velocity:Vector3f=<-1 0 0>) {isSetParameters=true, module=SetVariables_#4}
   /SetVariables(Particles.SpriteFacing=<0 -1 0>) {isSetParameters=true, module=SetVariables_#2}
   &DynamicMaterialParameters(Module.Param0WriteEnabled=true, Module.Param1WriteEnabled=true, Module.Param2WriteEnabled=true, Module.Param3WriteEnabled=true)
   &CameraOffset(Camera Offset Amount=-2)
   &ScaleColor(
    Scale Alpha=&FloatFromCurve(FloatCurve=DataInterface() {propertyValues {Curve {keys [<at 0.2 0>,<at 0.3 1>,<at 0.8 1>,<at 1 0>]}}})
    Scale RGB=&MakeVectorFromLinearColorRGB(Color=<rgba 1 1 1 1>)
    ScaleA=true
    ScaleRGB=true
   )
   &ScaleSpriteSize(Scale Factor=&Vector2DFromFloat(Value=1.5))
   &SolveForcesAndVelocity(Clamp Velocity=false, Limit Acceleration=false)
  ]
  rendererClasses [&NiagaraSpriteRendererProperties]
  renderers [{rendererClass=&NiagaraSpriteRendererProperties}]
 }
 Sprite_L_02001 {
  bEnabled=false
  emitterUpdateScript [
   &EmitterState()
   &SpawnRate(SpawnRate=&UniformRangedFloat(Minimum=10, Maximum=6), Use Spawn Probability=false)
  ]
  particleSpawnScript [
   &InitializeParticle(
    Color=$User.Secondary Color
    Lifetime=0.3
    Module.Write Color=true
    Module.Write Lifetime=true
    Module.Write Mass=true
    Module.Write Material Random=true
    Module.Write Position=true
    Module.Write Scale=false
    Module.Write SpriteRotation=true
    Module.Write SpriteSize=true
    Sprite Rotation=-5
    Sprite Size=&UniformRangedVector2D(Minimum=<10 22>, Maximum=<12 18>)
   )
   &SystemLocation(MaskSpawnGroup=false, Offset=<0 -1.5 8.2>)
   &AddVelocity(Velocity=&Multiply_Vector(A=<-130 1 -50>, B=&VectorFromFloat(Value=0.5)))
   /SetVariables(Particles.SubImageIndex=&UniformRangedFloat(Minimum=0, Maximum=2, Randomness Mode=::NewEnumerator1)) {isSetParameters=true, module=SetVariables_#5}
  ]
  particleUpdateScript [
   &ParticleState()
   &SpriteRotationRate(Rotation Rate=60)
   /SetVariables(Particles.Previous.Velocity:Vector3f=<-1 0 -0.05>) {isSetParameters=true, module=SetVariables_#6}
   /SetVariables(Particles.SpriteFacing=<0 -1 0>) {isSetParameters=true, module=SetVariables_#2}
   &DynamicMaterialParameters(Module.Param0WriteEnabled=true, Module.Param1WriteEnabled=true, Module.Param2WriteEnabled=true, Module.Param3WriteEnabled=true)
   &CameraOffset(Camera Offset Amount=-2.5)
   &ScaleColor(
    Scale Alpha=&FloatFromCurve(FloatCurve=DataInterface() {propertyValues {Curve {keys [<at 0 0>,<at 0.25 1>,<at 0.9 1>,<at 1 0>]}}})
    Scale RGB=&MakeVectorFromLinearColorRGB(Color=<rgba 1 1 1 1>)
    ScaleA=true
    ScaleRGB=true
   )
   &ScaleSpriteSize(Scale Factor=&Vector2DFromFloat(Value=1.4))
   &SolveForcesAndVelocity(Clamp Velocity=false, Limit Acceleration=false)
  ]
  rendererClasses [&NiagaraSpriteRendererProperties]
  renderers [{rendererClass=&NiagaraSpriteRendererProperties}]
 }
 Sprite_R_01001 {
  bEnabled=false
  emitterUpdateScript [
   &EmitterState()
   &SpawnRate(SpawnRate=7, Use Spawn Probability=false)
  ]
  particleSpawnScript [
   &InitializeParticle(
    Color=$User.Secondary Color
    Lifetime=0.4
    Module.Write Color=true
    Module.Write Lifetime=true
    Module.Write Mass=true
    Module.Write Material Random=true
    Module.Write Position=true
    Module.Write Scale=false
    Module.Write SpriteRotation=true
    Module.Write SpriteSize=true
    Sprite Size=&UniformRangedVector2D(Minimum=<8.5 23>, Maximum=<10 28>)
   )
   &SystemLocation(MaskSpawnGroup=false, Offset=<-12 -1.5 -10>)
   &AddVelocity()
   /SetVariables(Particles.SubImageIndex=&UniformRangedFloat(Minimum=0, Maximum=2, Randomness Mode=::NewEnumerator1)) {isSetParameters=true, module=SetVariables_#1}
  ]
  particleUpdateScript [
   &ParticleState()
   /SetVariables(Particles.Previous.Velocity:Vector3f=<-1 0 -0.03>) {isSetParameters=true, module=SetVariables_#4}
   /SetVariables(Particles.SpriteFacing=<0 -1 0>) {isSetParameters=true, module=SetVariables_#2}
   &DynamicMaterialParameters(Module.Param0WriteEnabled=true, Module.Param1WriteEnabled=true, Module.Param2WriteEnabled=true, Module.Param3WriteEnabled=true)
   &CameraOffset(Camera Offset Amount=-2)
   &ScaleColor(
    Scale Alpha=&FloatFromCurve(FloatCurve=DataInterface() {propertyValues {Curve {keys [<at 0.2 0>,<at 0.3 1>,<at 0.8 1>,<at 1 0>]}}})
    Scale RGB=&MakeVectorFromLinearColorRGB(Color=<rgba 1 1 1 1>)
    ScaleA=true
    ScaleRGB=true
   )
   &ScaleSpriteSize(Scale Factor=&Vector2DFromFloat(Value=1.5))
   &SolveForcesAndVelocity(Clamp Velocity=false, Limit Acceleration=false)
  ]
  rendererClasses [&NiagaraSpriteRendererProperties]
  renderers [{rendererClass=&NiagaraSpriteRendererProperties}]
 }
 Sprite_R_02001 {
  bEnabled=false
  emitterUpdateScript [
   &EmitterState()
   &SpawnRate(SpawnRate=&UniformRangedFloat(Minimum=10, Maximum=6), Use Spawn Probability=false)
  ]
  particleSpawnScript [
   &InitializeParticle(
    Color=$User.Secondary Color
    Lifetime=0.3
    Module.Write Color=true
    Module.Write Lifetime=true
    Module.Write Mass=true
    Module.Write Material Random=true
    Module.Write Position=true
    Module.Write Scale=false
    Module.Write SpriteRotation=true
    Module.Write SpriteSize=true
    Sprite Rotation=-5
    Sprite Size=&UniformRangedVector2D(Minimum=<10 22>, Maximum=<12 18>)
   )
   &SystemLocation(MaskSpawnGroup=false, Offset=<0 -1.5 -12.5>)
   &AddVelocity(Velocity=&Multiply_Vector(A=<-130 0 50>, B=&VectorFromFloat(Value=0.5)))
   /SetVariables(Particles.SubImageIndex=&UniformRangedFloat(Minimum=0, Maximum=2, Randomness Mode=::NewEnumerator1)) {isSetParameters=true, module=SetVariables_#5}
  ]
  particleUpdateScript [
   &ParticleState()
   &SpriteRotationRate(Rotation Rate=-60)
   /SetVariables(Particles.Previous.Velocity:Vector3f=<-1 0 -0.15>) {isSetParameters=true, module=SetVariables_#6}
   /SetVariables(Particles.SpriteFacing=<0 -1 0>) {isSetParameters=true, module=SetVariables_#2}
   &DynamicMaterialParameters(Module.Param0WriteEnabled=true, Module.Param1WriteEnabled=true, Module.Param2WriteEnabled=true, Module.Param3WriteEnabled=true)
   &CameraOffset(Camera Offset Amount=-2.5)
   &ScaleColor(
    Scale Alpha=&FloatFromCurve(FloatCurve=DataInterface() {propertyValues {Curve {keys [<at 0 0>,<at 0.25 1>,<at 0.9 1>,<at 1 0>]}}})
    Scale RGB=&MakeVectorFromLinearColorRGB(Color=<rgba 1 1 1 1>)
    ScaleA=true
    ScaleRGB=true
   )
   &ScaleSpriteSize(Scale Factor=&Vector2DFromFloat(Value=1.4))
   &SolveForcesAndVelocity(Clamp Velocity=false, Limit Acceleration=false)
  ]
  rendererClasses [&NiagaraSpriteRendererProperties]
  renderers [{rendererClass=&NiagaraSpriteRendererProperties}]
 }
 Top_Streaks {
  emitterUpdateScript [
   &EmitterState(Life Cycle Mode=::NewEnumerator1, Loop Duration=1, UseLoopDelay=false)
   &SpawnRate(SpawnRate=12, Use Spawn Probability=false)
  ]
  particleSpawnScript [
   &InitializeParticle(
    Color=$User.Primary Color
    Mesh Scale=&UniformRangedVector(Minimum=<10 15 0>, Maximum=<10 15 0>)
    Module.Write Color=true
    Module.Write Lifetime=true
    Module.Write Mass=true
    Module.Write Material Random=true
    Module.Write Position=true
    Module.Write Scale=true
    Module.Write SpriteRotation=true
    Module.Write SpriteSize=true
    Sprite Size=&UniformRangedVector2D(Minimum=<10 10>, Maximum=<10 12>)
   )
   &FN_SetSpawnGroup_v2(Number of Spawn Groups=2)
   /SetVariables(Particles.SubImageIndex=&UniformRangedFloat(Minimum=0, Maximum=1, Randomness Mode=::NewEnumerator1)) {isSetParameters=true, module=SetVariables_#1}
   /SetVariables(Particles.SpriteFacing=&TransformVector(Vector=<0 1 0>, Destination Space=::World)) {enabled=false, isSetParameters=true, module=SetVariables_#2}
   &SpriteFacingAndAlignment(Sprite Alignment=<-1 1 0>, Sprite Facing=<0 1 0>, WriteAlignment=true, WriteFacing=true)
  ]
  particleUpdateScript [
   &ParticleState()
   /SetVariables(Particles.SpriteRotation:NiagaraFloat=&MaskFloatBySpawnGroup(
     Masked Float=&FloatFromCurve(FloatCurve=DataInterface() {propertyValues {Curve {keys [<at 0 80>,<at 0.1931 50.6864>,<at 0.5 0>,<at 1 -20>]}}})
     Passthrough Float=&FloatFromCurve(FloatCurve=DataInterface() {propertyValues {Curve {keys [<at 0 -80>,<at 0.1931 -50>,<at 0.5 0>,<at 1 20>]}}})
    )) {isSetParameters=true, module=SetVariables_#3}
   &SkeletalMeshLocation(
    Bone / Socket Index=&Add_Integer(A=&SpawnGroup(), B=3)
    Inherit Velocity=false
    MaskSpawnGroup=false
    Mesh Sampling Type=::NewEnumerator5
    Sampled Position Offset=&MakeVector(
     X=&FloatFromCurve(FloatCurve=DataInterface() {propertyValues {Curve {keys [<at 0 1>,<at 1 30>]}}})
     Y=0
     Z=&MaskFloatBySpawnGroup(
      Masked Float=&FloatFromCurve(FloatCurve=DataInterface() {propertyValues {Curve {keys [<at 0 0 RCIM_Cubic>,<at 0.5282 -8 RCIM_Cubic>,<at 1 0 RCIM_Cubic>]}}})
      Passthrough Float=&FloatFromCurve(FloatCurve=DataInterface() {propertyValues {Curve {keys [<at 0 0 RCIM_Cubic>,<at 0.5282 8 RCIM_Cubic>,<at 1 0 RCIM_Cubic>]}}})
     )
    )
    Sampled Position Offset Coordinate Space:ENiagaraMeshTransforms=::NewEnumerator3
    Skeletal Mesh=$User.SkeletalMesh
    Socket Sampling Mode=::NewEnumerator5
   )
   /SetVariables(Particles.Previous.Velocity:Vector3f=<-1 0 0>) {isSetParameters=true, module=SetVariables_#4}
   /SetVariables(Particles.StreaksAlignment:Vector3f=&TransformVector(Vector=<-1 0 -0.05>, Destination Space=::World)) {enabled=false, isSetParameters=true}
   &DynamicMaterialParameters(
    Index 0 Param 1=&MaskFloatBySpawnGroup(
     Masked Float=&Lerp_Float(A=-1, B=0, Alpha=$Particles.NormalizedAge)
     Passthrough Float=&Lerp_Float(A=1, B=0, Alpha=$Particles.NormalizedAge)
    )
    Module.Param0WriteEnabled=true
    Module.Param1WriteEnabled=true
    Module.Param2WriteEnabled=true
    Module.Param3WriteEnabled=true
   )
   &ScaleSpriteSize(Scale Factor=&Vector2DFromFloat(Value=&FloatFromCurve(FloatCurve=DataInterface() {propertyValues {Curve {keys [<at 0 0>,<at 0.2 1.5>]}}})))
   &SolveForcesAndVelocity(Clamp Velocity=false, Limit Acceleration=false)
   &Color(Color=&LerpLinearColors(StartColor=$User.Primary Color, EndColor=$User.Secondary Color, LerpFactor=$Particles.NormalizedAge))
  ]
  rendererClasses [&NiagaraSpriteRendererProperties]
  renderers [{rendererClass=&NiagaraSpriteRendererProperties}]
 }
 Wrist_Airglow {
  emitterUpdateScript [
   &EmitterState(Inactive Response=::NewEnumerator1, Life Cycle Mode=::NewEnumerator1, Loop Duration=0.25, UseLoopDelay=false)
   &SpawnBurst_Instantaneous(Spawn Count=1, Use Spawn Probability=false)
  ]
  particleSpawnScript [
   &InitializeParticle(
    Color=$User.Primary Color
    Lifetime=0.25
    Module.Write Color=true
    Module.Write Lifetime=true
    Module.Write Mass=false
    Module.Write Material Random=true
    Module.Write Position=false
    Module.Write Scale=false
    Module.Write SpriteRotation=false
    Module.Write SpriteSize=true
    Position=&SimulationPosition()
    Sprite Size=&Vector2DFromFloat(Value=45)
   )
   &DynamicMaterialParameters(
    Index 0 Param 1=1
    Module.Param0WriteEnabled=true
    Module.Param1WriteEnabled=true
    Module.Param2WriteEnabled=true
    Module.Param3WriteEnabled=true
   )
   &SystemLocation(MaskSpawnGroup=false, Offset=<10 -2.5 -2.5>)
  ]
  particleUpdateScript [
   &ParticleState()
   &CameraOffset(Camera Offset Amount=44)
   &ScaleColor(Scale Alpha=0.4, Scale RGB=&VectorFromFloat(Value=0.8), ScaleA=true, ScaleRGB=true)
  ]
  rendererClasses [&NiagaraSpriteRendererProperties]
  renderers [{rendererClass=&NiagaraSpriteRendererProperties}]
 }
}
summary {
 systemName=NS_Pickaxe_Elastic_Cosmic_Idle
 emitters [
  {emitterName=Bottom_Streaks, rendererClasses [&NiagaraSpriteRendererProperties]}
  {emitterName=Middle_Streaks, rendererClasses [&NiagaraSpriteRendererProperties]}
  {emitterName=Light, rendererClasses [&NiagaraLightRendererProperties]}
  {bEnabled=false, emitterName=Decal_Glow, rendererClasses [&NiagaraSpriteRendererProperties]}
  {emitterName=Decal_Smoke, rendererClasses [&NiagaraSpriteRendererProperties]}
  {emitterName=Particle_Spark, rendererClasses [&NiagaraSpriteRendererProperties]}
  {bEnabled=false, emitterName=Sprite_L_01001, rendererClasses [&NiagaraSpriteRendererProperties]}
  {bEnabled=false, emitterName=Sprite_R_01001, rendererClasses [&NiagaraSpriteRendererProperties]}
  {bEnabled=false, emitterName=Sprite_L_02001, rendererClasses [&NiagaraSpriteRendererProperties]}
  {bEnabled=false, emitterName=Sprite_R_02001, rendererClasses [&NiagaraSpriteRendererProperties]}
  {emitterName=Decal_Smoke001, rendererClasses [&NiagaraSpriteRendererProperties]}
  {emitterName=Particle_Spark001, rendererClasses [&NiagaraSpriteRendererProperties]}
  {emitterName=Lingering_Sparks, rendererClasses [&NiagaraSpriteRendererProperties]}
  {emitterName=Blade_Airglow, rendererClasses [&NiagaraMeshRendererProperties]}
  {emitterName=Mesh_Sword, rendererClasses [&NiagaraMeshRendererProperties]}
  {emitterName=Top_Streaks, rendererClasses [&NiagaraSpriteRendererProperties]}
  {emitterName=Wrist_Airglow, rendererClasses [&NiagaraSpriteRendererProperties]}
 ]
 userVariables [
  User.Primary Color {type=LinearColor, description="", defaultValue {struct=LinearColor, value=<rgba 0 0.1333 1 1>}}
  User.Secondary Color {type=LinearColor, description="", defaultValue {struct=LinearColor, value=<rgba 0.5 0 1 1>}}
  User.SkeletalMesh {
   type=NiagaraDataInterfaceSkeletalMesh
   defaultValue {
    struct=NiagaraExt_VariableValue_DataInterface
    value {dataInterface=&self:NiagaraDataInterfaceSkeletalMesh_1}
   }
  }
 ]
}
