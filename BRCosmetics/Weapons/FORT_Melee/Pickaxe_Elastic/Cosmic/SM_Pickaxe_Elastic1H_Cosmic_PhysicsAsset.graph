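# Physics asset for the Cosmic variant of the one-handed Elastic pickaxe, a melee weapon cosmetic.
# Its constraint list is empty, so it defines no joints between physics bodies.
# No body setups, collision shapes or other non-default settings appear.
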PhysicsAsset /BRCosmetics/Weapons/FORT_Melee/Pickaxe_Elastic/Cosmic/SM_Pickaxe_Elastic1H_Cosmic_PhysicsAsset
constraints []
